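; Hologram capture-stage light blueprint that sets a point light's appearance according to the current stage.
; SetLightStage switches on Stage, overridden by DebugStage when that is zero or above.
; Stages 1 to 3 show the light with their own color and fixed intensity, while stage 0 and other values hide it.
; It runs from the construction script, BeginPlay and OnRep_Stage, so replicated stage changes update the light without ticking.
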(event EventBeginPlay
  (CallFunction|SetLightStage))

(event Collision|EventActorBeginOverlap
  (params (OtherActor :type Actor%20Object%20Reference)))

(event EventTick
  (params (DeltaSeconds :type Float%20%28single-precision%29)))

(fn OnRep_Stage
  (CallFunction|SetLightStage))

(fn SetLightStage
  (bind _pointlightcomponent (Variables|Light|GetPointLightComponent))
  (bind _pointlightcomponent_2 (Variables|Light|GetPointLightComponent))
  (bind _pointlightcomponent_3 (Variables|Light|GetPointLightComponent))
  (switch int (Math|Integer|SelectInt (Variables|Default|GetDebugStage) (Variables|Default|GetStage) (>= (Variables|Default|GetDebugStage) 0))
    (:Default
      _)
    (:0
      _)
    (:1
      (Rendering|SetVisibility (Variables|Light|GetPointLightComponent) true false)
      (Rendering|Components|Light|SetLightColor _pointlightcomponent_3 (Math|Conversions|ToLinearColor(Color) (Variables|Default|GetLight%20Color%20Stage%201)) true)
      (Rendering|Components|Light|SetIntensity _pointlightcomponent_3 22.0))
    (:2
      (Rendering|SetVisibility (Variables|Light|GetPointLightComponent) true false)
      (Rendering|Components|Light|SetLightColor _pointlightcomponent_2 (Math|Conversions|ToLinearColor(Color) (Variables|Default|GetLight%20Color%20Stage%202)) true)
      (Rendering|Components|Light|SetIntensity _pointlightcomponent_2 6.0))
    (:3
      (Rendering|SetVisibility (Variables|Light|GetPointLightComponent) true false)
      (Rendering|Components|Light|SetLightColor _pointlightcomponent (Math|Conversions|ToLinearColor(Color) (Variables|Default|GetLight%20Color%20Stage%203)) true)
      (Rendering|Components|Light|SetIntensity _pointlightcomponent 5.0)))
  (Rendering|SetVisibility (Variables|Light|GetPointLightComponent) false false))

(fn ConstructionScript
  (CallFunction|SetLightStage))
