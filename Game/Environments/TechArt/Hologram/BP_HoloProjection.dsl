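; Hologram projection actor that pairs a character mesh with a holo projector cone, configured entirely in the construction script.
; It applies an optional character mesh with material overrides, then feeds projection source, target and cone size to that mesh's materials.
; Cone width and length derive from the source transform's scale and the actor's X scale.
; The projector is aimed along the source-to-target direction at unit world scale, with cone size and actor scale passed to its materials.
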(event EventBeginPlay)

(event Collision|EventActorBeginOverlap
  (params (OtherActor :type Actor%20Object%20Reference)))

(event EventTick
  (params (DeltaSeconds :type Float%20%28single-precision%29)))

(fn ConstructionScript
  (bind _projectionsource (Variables|Default|GetProjectionSource))
  (bind _scale (.scale _projectionsource))
  (bind _returnvalue 5e+02)
  (bind _returnvalue_2 (Transformation|GetActorScale3D))
  (bind _x (.x _returnvalue_2))
  (bind _returnvalue_3 2.5e+02)
  (bind _x_2 (.x _scale))
  (bind _z (.z _scale))
  (bind _returnvalue_4 (Utilities|Operators|Multiply _returnvalue _z _x))
  (bind _returnvalue_5 (Utilities|Operators|Multiply _returnvalue_3 _x_2 _z _x))
  (bind _sm_character (Variables|Default|GetSM_Character))
  (bind _sm_holoprojector (Variables|Default|GetSM_HoloProjector))
  (Utilities|FlowControl|Sequence
    (:then_0
      (bind _staticmeshcharacter (Variables|Default|GetStaticMeshCharacter)
        (:then
          (Components|StaticMesh|SetStaticMesh (Variables|Default|GetSM_Character) _staticmeshcharacter)
          (for (_array_element _array_index) (Variables|Default|GetMaterialOverride)
            (Rendering|Material|SetMaterial (Variables|Default|GetSM_Character) _array_index _array_element)))
        (:else)))
    (:then_2
      (Rendering|Material|SetVectorParameterValueonMaterials _sm_character "ProjectionTarget" (Variables|Default|GetProjectionTarget))
      (Rendering|Material|SetVectorParameterValueonMaterials _sm_character "ProjectionSource" (.location (Variables|Default|GetProjectionSource)))
      (Rendering|Material|SetScalarParameterValueonMaterials _sm_character "ConeWidth" _returnvalue_5)
      (Rendering|Material|SetScalarParameterValueonMaterials _sm_character "ConeLength" _returnvalue_4))
    (:then_3
      (Transformation|SetRelativeTransform _sm_holoprojector (Math|Transform|MakeTransform (.location _projectionsource) (Math|Rotator|MakeRotfromZ (Math|Vector|GetUnitDirection(Vector) (.location _projectionsource) (Variables|Default|GetProjectionTarget)))) false false)
      (Transformation|SetWorldScale3D _sm_holoprojector "1.000000,1.000000,1.000000")
      (Rendering|Material|SetScalarParameterValueonMaterials _sm_holoprojector "ConeWidth" _returnvalue_5)
      (Rendering|Material|SetScalarParameterValueonMaterials _sm_holoprojector "ConeLength" _returnvalue_4)
      (Rendering|Material|SetScalarParameterValueonMaterials _sm_holoprojector "ActorScale" _x))))
