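# Additive unlit material for a hologram projector, with static switches selecting its cone, laser or flare variant.
# Cone and lasers combine time-of-day brightness with panning noise and are extended by world position offset; lasers use SplineThicken and vertex-color-randomized sine sway.
# The flare is a billboard sprite with counter-rotating textures and depth fade.
# MPC_Hologram supplies color switch, contested tint and FadeOpacity, which scales opacity and selects the vertex offset.
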Material /Game/Environments/TechArt/Hologram/M_HologramProjector
graph {
 properties {
  bAutomaticallySetUsageInEditor=false
  bUseMaterialAttributes=true
  blendMode=BLEND_Additive
  pixelDepthOffsetMode=PDOM_Legacy
  shadingModel=MSM_Unlit
  usedShadingModels=Unlit
 }
 nodes [
  n1 = LinearInterpolate(
   A=VectorParameter() {parameterName=Laser_Color, defaultValue=<rgba 0.0908 0.1412 0.1875 1>}.RGB
   B=VectorParameter() {parameterName=Laser_Color_2, defaultValue=<rgba 0.4791 0.5104 0 1>}.RGB
   Alpha=NamedRerouteUsage()
  )
  n2 = ScalarParameter() {defaultValue=0.07, parameterName=Brightness_Lasers_Evening}
  n3 = MaterialFunctionCall(
   Value_Morning=@n2
   Value_Day=ScalarParameter() {defaultValue=0.1, parameterName=Brightness_Lasers}
   Value_Evening=@n2
   Value_Night=ScalarParameter() {defaultValue=0.05, parameterName=Brightness_Lasers_Night}
  ) {materialFunction=&Scalar_Time_of_Day}
  n4 = LinearInterpolate(
   A=VectorParameter() {parameterName=Cone_Color, defaultValue=<rgba 0.2075 0.3436 0.4688 1>}.RGB
   B=VectorParameter() {parameterName=Cone_Color_2, defaultValue=<rgba 0.4688 0.4376 0 1>}.RGB
   Alpha=NamedRerouteUsage()
  )
  n5 = ScalarParameter() {defaultValue=0.25, parameterName=Brightness_Cone_Night}
  n6 = MaterialFunctionCall(Value_Morning=@n5, Value_Day=ScalarParameter() {defaultValue=0.5, parameterName=Brightness_Cone}, Value_Evening=@n5, Value_Night=@n5) {materialFunction=&Scalar_Time_of_Day}
  n7 = TextureCoordinate()
  n8 = AppendVector(A=ScalarParameter() {defaultValue=2, parameterName=Noise_Scale_u}, B=ScalarParameter() {defaultValue=-0.0537, parameterName=Noise_Scale_V})
  n9 = ScalarParameter() {defaultValue=0.44, parameterName=Noise_Panner_u}
  n10 = ScalarParameter() {defaultValue=-0.124, parameterName=Noise_Panner_v}
  n11 = TextureSample(UVs=Panner(Coordinate=Multiply(A=@n7, B=@n8), Speed=AppendVector(A=@n9, B=@n10))) {samplerType=SAMPLERTYPE_LinearColor, texture=/Game/Effects/Fort_Effects/Gameplay/SmokeDustExplosions/Textures/T_NoisePack_LinearColor}
  n12 = Multiply(
   A=StaticSwitchParameter(True=Multiply(A=@n1, B=@n3.Result), False=Multiply(A=@n4, B=@n6.Result)) {parameterName=IsLasers}
   B=Desaturation(None=@n11.RGB, Fraction=ScalarParameter() {defaultValue=0.224, parameterName=DesatRandomColors})
  )
  Multiply(A=@n12, B=Constant())
  n13 = ScalarParameter() {defaultValue=4, parameterName=WidthBase}
  n14 = ScalarParameter() {defaultValue=1, parameterName=ActorScale}
  Multiply(A=@n13, B=@n14)
  n15 = TransformPosition(None=Constant3Vector()) {transformType=TRANSFORMPOSSOURCE_World}
  NamedRerouteDeclaration(None=@n15) {name=OriginPivot, nodeColor=<rgba 1 0 0.1875 1>}
  NamedRerouteDeclaration(None=CollectionParameter() {collection=&/MPC_Hologram, parameterName=ColorSwitch}) {name=ColorSwitch, nodeColor=<rgba 0 0.2266 1 1>}
  NamedRerouteDeclaration(None=CollectionParameter() {collection=&MPC_Hologram, parameterName=Contested}) {name=Contested, nodeColor=<rgba 0.8281 1 0 1>}
  Reroute(None=@n15)
  n16 = TextureSample(UVs=Rotator() {speed=0.5}) {texture=/Game/Effects/Fort_Effects/Textures/Flares/T_Flare_Spikeball}
  n17 = Add(
   A=LinearInterpolate(
    A=VectorParameter() {parameterName=Flare_Color, defaultValue=<rgba 1 1 1 1>}.RGB
    B=VectorParameter() {parameterName=Flare_Color_2, defaultValue=<rgba 5 1.4252 0 1>}.RGB
    Alpha=NamedRerouteUsage()
   )
   B=@n16.RGB
  )
  n18 = Add(A=@n16.RGB, B=TextureSample(UVs=Rotator() {speed=-0.6}) {texture=&T_Flare_Spikeball}.RGB)
  n19 = Multiply(
   A=DepthFade(Opacity=ComponentMask(None=@n18) {r=true}, FadeDistance=ScalarParameter() {defaultValue=100, parameterName=FlareDepthFade})
   B=SphereMask(
    A=TextureCoordinate()
    B=Constant() {r=0.5}
    Radius=ScalarParameter() {defaultValue=10, parameterName=FlareSphereMaskRadius}
    Hardness=ScalarParameter() {parameterName=FlareSphereMaskHardness}
   ) {hardnessPercent=0}
  )
  n20 = MaterialFunctionCall!(
   Center Location=NamedRerouteUsage()
   0-1 UVs
   XY scale=ScalarParameter() {defaultValue=89.126, parameterName=FlareScale}
   OverridePivot
   "Normals (Optional)"
  ) {materialFunction=&Sprite}
  n21 = LinearInterpolate(A=@n12, B=Multiply(A=@n12, B=VectorParameter() {parameterName=Contested, defaultValue=<rgba 5 0 0 1>}.RGB), Alpha=NamedRerouteUsage())
  n22 = LinearInterpolate(A=ScalarParameter() {defaultValue=0.341, parameterName=Min}, B=ScalarParameter() {defaultValue=1.2276, parameterName=Max}, Alpha=@n11.G)
  n23 = ComponentMask(None=TextureCoordinate()) {r=true}
  n24 = Multiply(A=@n23, B=OneMinus(None=@n23))
  n25 = ComponentMask(None=TextureCoordinate()) {g=true}
  n26 = Power(Base=OneMinus(None=@n25), Exp=ScalarParameter() {defaultValue=1.3342, parameterName=FallOff})
  n27 = Multiply(A=Power(Base=Multiply(A=@n24) {constB=4}, Exp=ScalarParameter() {defaultValue=1.354, parameterName=FallloffLaser}), B=@n26)
  n28 = Subtract(A=TransformPosition(None=WorldPosition().XYZ) {transformSourceType=TRANSFORMPOSSOURCE_World}, B=Constant3Vector())
  n29 = Constant3Vector() {constant=<rgba 1 1 0 0>}
  n30 = Transform(None=Normalize(VectorInput=Multiply(A=@n28, B=@n29))) {transformSourceType=TRANSFORMSOURCE_Local}
  n31 = Abs(None=DotProduct(A=Normalize(VectorInput=@n30), B=CameraVectorWS()))
  n32 = ScalarParameter() {defaultValue=3.7955, parameterName=Fresnel_Power}
  n33 = Multiply(A=Multiply(A=Power(Base=@n31, Exp=@n32), B=@n26), B=ScalarParameter() {defaultValue=1, parameterName=ConeOpacity})
  n34 = DepthFade(
   Opacity=Multiply(A=Saturate(None=@n22), B=StaticSwitchParameter(True=@n27, False=@n33) {parameterName=IsLasers})
   FadeDistance=ScalarParameter() {defaultValue=100, parameterName=FlareDepthFade}
  )
  n35 = Constant3Vector() {constant=<rgba 0 0 1 0>}
  n36 = Multiply(
   A=Multiply(A=@n25, B=Transform(None=@n35) {transformSourceType=TRANSFORMSOURCE_Local})
   B=ScalarParameter() {defaultValue=494.5073, parameterName=ConeLength}
  )
  n37 = ScalarParameter() {defaultValue=10, parameterName=WidthTip}
  n38 = WorldPosition()
  n39 = VertexColor()
  n40 = Multiply(A=@n39.R, B=ScalarParameter() {defaultValue=6.378, parameterName=Random}) {constB=33.33}
  n41 = MaterialFunctionCall(Float3=TransformPosition(None=Constant3Vector() {constant=<rgba 0.15 0.24 0.5 0>}) {transformType=TRANSFORMPOSSOURCE_World}) {materialFunction=&BreakOutFloat3Components}
  n42 = Add(A=Add(A=@n41.R, B=@n41.R), B=@n41.B)
  n43 = Add(A=Time(), B=Multiply(A=Frac(None=@n42), B=ScalarParameter() {defaultValue=33.33, parameterName=Random3}))
  n44 = ScalarParameter() {defaultValue=0.55, parameterName=Speed}
  n45 = Add(A=@n44) {constB=-0.01}
  n46 = Sine(None=Multiply(A=Add(A=@n43, B=@n40), B=LinearInterpolate(A=@n45, B=@n44, Alpha=@n39.R)))
  n47 = Multiply(A=@n39.R, B=ScalarParameter() {defaultValue=34.3002, parameterName=Random2}) {constB=33.33}
  n48 = ScalarParameter() {defaultValue=0.576, parameterName=Speed2}
  n49 = Add(A=@n48) {constB=-0.01}
  n50 = Sine(None=Multiply(A=Add(A=@n43, B=@n47), B=LinearInterpolate(A=@n49, B=@n48, Alpha=@n39.R)))
  n51 = Transform(None=AppendVector(A=AppendVector(A=@n46, B=@n50), B=Constant())) {transformSourceType=TRANSFORMSOURCE_Local}
  n52 = ComponentMask(None=TextureCoordinate()) {g=true}
  n53 = ScalarParameter() {defaultValue=245.0628, parameterName=ConeWidth}
  n54 = Add(
   A=MaterialFunctionCall(WidthBase=Multiply(A=@n37, B=@n14), WidthTip=@n13, WorldPosition=Add(A=@n36, B=@n38.XYZ)) {materialFunction=&SplineThicken}.WorldPositionOffset
   B=Multiply(A=Multiply(A=@n51, B=@n52), B=Divide(A=@n53))
  )
  n55 = Multiply(A=@n25, B=@n30)
  n56 = ScalarParameter() {defaultValue=245.0628, parameterName=ConeWidth}
  n57 = StaticSwitchParameter(
   True=Multiply(A=Add(A=@n36, B=@n54), B=ScalarParameter() {defaultValue=0.75, parameterName=Laser_Brightness})
   False=Add(A=@n36, B=Multiply(A=@n55, B=@n56))
  ) {parameterName=IsLasers}
  n58 = GetMaterialAttributes(None=StaticSwitchParameter(
    True=SetMaterialAttributes!(MaterialAttributes, Emissive Color=@n17, Opacity=@n19, World Position Offset=@n20.World position offset) {attributeSetTypes [#1, #2, #3]}
    False=SetMaterialAttributes(Emissive Color=@n21, Opacity=@n34, World Position Offset=@n57) {attributeSetTypes [#1, #2, #3]}
   ) {parameterName=IsFlare}) {attributeGetTypes [#2, #3]}
  n59 = CollectionParameter() {collection=&MPC_Hologram, parameterName=FadeOpacity}
  n60 = Subtract(A=WorldPosition().XYZ, B=ObjectPositionWS())
  SetMaterialAttributes!(
   MaterialAttributes=@n58.MaterialAttributes
   Opacity=Multiply(A=@n58.Opacity, B=@n59)
   World Position Offset=If(A=@n59, "A > B"=@n58.World Position Offset, "A == B"=@n60, "A < B"=@n60)
  ) {attributeSetTypes [#2, #3]}
  TextureCoordinate()
  Time()
 ]
}
